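# Translucent unlit stylized flame material for the SulfurDean pickaxe's Niagara sprite and mesh particles, with scalar blends for bowl, ember and glow variants.
# Panned, distorted noise plus gradient and taper masks are rounded into hard-edged opacity and color bands over gradient colors, tinted by particle color.
# Depth and view-angle fades soften opacity.
# World position offset combines spline thickening with a clamped, UV-weighted bend driven by Niagara dynamic parameters.
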Material /BRCosmetics/Weapons/FORT_Melee/Pickaxe_SulfurDean/FX/M_Pickaxe_SulfurDean_StylizedFlame
graph {
 properties {
  bAutomaticallySetUsageInEditor=false
  bUsedWithNiagaraMeshParticles=true
  bUsedWithNiagaraSprites=true
  blendMode=BLEND_Translucent
  shadingModel=MSM_Unlit
  twoSided=true
  usedShadingModels=Unlit
 }
 nodes [
  n1 = ScalarParameter() {defaultValue=9.9953, parameterName=Spline_Base}
  n2 = ScalarParameter() {defaultValue=9.6805, parameterName=Spline_Tip}
  n3 = StaticBool()
  n4 = LinearInterpolate(
   A=Constant()
   B=Multiply(
    A=ScalarParameter() {defaultValue=1, parameterName=Spline_ThickMul}
    B=MaterialFunctionCall(WidthBase=@n1, WidthTip=@n2, Expand U or V UV Channel=@n3) {materialFunction=&SplineThicken}.WorldPositionOffset
   )
   Alpha=ScalarParameter() {parameterName=Use_SplineThicken}
  )
  n5 = DynamicParameter() {desc=Parameters set by Niagara, parameterIndex=1, paramNames [X,Y,Z,WPOIntensity]}
  n6 = Subtract(
   A=MaterialFunctionCall(X=@n5.X, Y=@n5.X, Z=@n5.X) {materialFunction=&MakeFloat3}.Result
   B=ParticlePositionWS()
  )
  n7 = ScalarParameter() {defaultValue=15, parameterName=Interpolate_WPO_Range}
  n8 = Multiply(A=@n7) {constB=-1}
  n9 = Power(Base=ComponentMask(None=TextureCoordinate()) {r=true}, Exp=ScalarParameter() {defaultValue=2, parameterName=WPOBendPower})
  n10 = LinearInterpolate(
   A=Constant()
   B=LinearInterpolate(B=Clamp(None=@n6, Min=@n8, Max=@n7) {maxDefault=50, minDefault=-50}, Alpha=Multiply(A=@n9, B=@n5.WPOIntensity))
   Alpha=ScalarParameter() {defaultValue=1, parameterName=Use interpolate WPO}
  )
  n11 = ScalarParameter() {parameterName=LerpBowl}
  LinearInterpolate(A=LinearInterpolate(A=Add(A=@n4, B=@n10), Alpha=@n11) {constB=0}, Alpha=ScalarParameter() {parameterName=LerpEmber}) {constB=0}
  n12 = ParticleColor()
  n13 = TextureCoordinate()
  n14 = LinearInterpolate(A=ComponentMask(None=@n13) {r=true}, B=ComponentMask(None=@n13) {g=true}, Alpha=ScalarParameter() {parameterName=LerpBowl})
  n15 = LinearInterpolate(
   A=VectorParameter() {parameterName=Color_1, defaultValue=<rgba 2 0.5075 0 1>}.RGB
   B=VectorParameter() {parameterName=Color_3, defaultValue=<rgba 2 0.1034 0 1>}.RGB
   Alpha=@n14
  )
  n16 = TextureCoordinate()
  n17 = ComponentMask(None=@n16) {g=true}
  n18 = Saturate(None=@n17)
  n19 = LinearInterpolate(A=ScalarParameter() {defaultValue=-0.672, parameterName=SphereMaskUVDistortion}, Alpha=ScalarParameter() {parameterName=LerpEmber}) {constB=0}
  n20 = ScalarParameter() {parameterName=LerpBowl}
  n21 = Multiply(
   A=SphereMask(A=@n16, B=Constant2Vector() {g=0.5}, Radius=Constant() {r=1}, Hardness=Constant())
   B=Multiply(A=LinearInterpolate(Alpha=@n18) {constA=1, constB=-1}, B=LinearInterpolate(A=@n19, Alpha=@n20) {constB=0})
  ) {constB=0.128}
  n22 = DynamicParameter() {paramNames [Offset1,Offset2,Param3,Param4]}
  n23 = Multiply(
   A=Constant2Vector() {g=0.8, r=0.6}
   B=AppendVector(A=ScalarParameter() {defaultValue=0.744, parameterName=Tile_Mult_1}, B=ScalarParameter() {defaultValue=1.4485, parameterName=Tile_Mult_2})
  )
  n24 = Multiply(A=Constant2Vector() {g=0.3, r=2}, B=ScalarParameter() {defaultValue=1, parameterName=BowlTile_Mult})
  n25 = ScalarParameter() {parameterName=LerpBowl}
  n26 = Constant2Vector() {g=0.5, r=0.7}
  n27 = ScalarParameter() {defaultValue=1, parameterName=EmberTileMult}
  n28 = ScalarParameter() {parameterName=LerpEmber}
  n29 = Multiply(
   A=Add(A=Add(A=@n13, B=@n21), B=AppendVector(A=@n22.Offset1, B=@n22.Offset1))
   B=LinearInterpolate(A=LinearInterpolate(A=@n23, B=@n24, Alpha=@n25), B=Multiply(A=@n26, B=@n27), Alpha=@n28)
  )
  n30 = Constant2Vector() {g=0.1, r=-0.8}
  n31 = Constant2Vector() {g=0.3, r=0.05}
  n32 = Multiply(
   A=LinearInterpolate(A=LinearInterpolate(A=@n30, B=@n31, Alpha=@n25), B=Constant2Vector() {g=0.15, r=0.02}, Alpha=@n28)
   B=ScalarParameter() {defaultValue=1, parameterName=SpeedMult}
  )
  n33 = StaticComponentMaskParameter(None=TextureSampleParameter2D(UVs=Panner(Coordinate=@n29, Speed=@n32) {speedX=-0.5}) {parameterName=Texture, texture=&T_NoisePack_01}.RGB) {defaultG=true, parameterName=Texture_Mask}
  n34 = ComponentMask(None=@n13) {r=true}
  n35 = ScalarParameter() {parameterName=LerpBowl}
  n36 = Multiply(
   A=LinearInterpolate(A=OneMinus(None=@n34), B=ComponentMask(None=@n13) {g=true}, Alpha=@n35)
   B=ScalarParameter() {defaultValue=0.4673, parameterName=GradientAdd_Mult}
  )
  n37 = ComponentMask(None=@n13) {g=true}
  n38 = Multiply(A=Sine(None=@n37) {period=2}, B=Sine(None=ComponentMask(None=@n13) {r=true}) {period=2})
  n39 = Multiply(
   A=Multiply(A=Add(A=@n33, B=@n36), B=LinearInterpolate(A=@n38, B=@n37, Alpha=@n35))
   B=ScalarParameter() {defaultValue=1.5104, parameterName=Tex_Mult}
  ) {constB=24}
  n40 = MaterialFunctionCall(UVs=@n13, "Rotation Angle (0-1)"=Constant() {r=-0.05}) {materialFunction=&CustomRotator}
  n41 = MaterialFunctionCall(UVs=@n13, "Rotation Angle (0-1)"=Constant() {r=-0.447}) {materialFunction=&CustomRotator}
  n42 = Multiply(
   A=ComponentMask(None=Add(A=@n40.Rotated Values) {constB=-0.25}) {g=true}
   B=ComponentMask(None=Add(A=@n41.Rotated Values) {constB=-0.25}) {g=true}
  )
  n43 = LinearInterpolate(B=Saturate(None=Multiply(A=@n42) {constB=30.9627}), Alpha=ScalarParameter() {defaultValue=1, parameterName=LerpTaper}) {constA=1}
  n44 = ScalarParameter() {parameterName=LerpEmber}
  n45 = Multiply(A=@n39, B=LinearInterpolate(A=LinearInterpolate(A=@n43, Alpha=@n44), Alpha=ScalarParameter() {parameterName=LerpBowl}))
  n46 = Subtract(A=@n45, B=ScalarParameter() {defaultValue=0.3881, parameterName=Color_Subtr})
  n47 = LinearInterpolate(
   A=@n15
   B=LinearInterpolate(
    A=VectorParameter() {parameterName=Color_2, defaultValue=<rgba 0.9911 0.3763 0.0091 1>}.RGB
    B=VectorParameter() {parameterName=Color_4, defaultValue=<rgba 0.9387 0.0557 0 1>}.RGB
    Alpha=@n14
   )
   Alpha=Saturate(None=Round(None=@n46))
  )
  n48 = ScalarParameter() {parameterName=Lerp_OneColor}
  MaterialFunctionCall(Emissive=Multiply(A=@n12.RGB, B=LinearInterpolate(A=@n47, B=@n15, Alpha=@n48))) {materialFunction=&MF_FX_EmissiveAdjustments}
  n49 = VectorParameter() {parameterName=ViewBasedFadeLocalAxis, defaultValue=<rgba 0 0 1 1>}
  n50 = CameraVectorWS()
  n51 = Abs(None=DotProduct(A=Transform(None=@n49.RGB) {transformSourceType=TRANSFORMSOURCE_Local}, B=@n50))
  n52 = VertexNormalWS()
  Min(A=@n51, B=Abs(None=DotProduct(A=@n52, B=@n50)))
  n53 = Round(None=@n45)
  n54 = ScalarParameter() {parameterName=LerpGlow}
  n55 = DepthFade(
   Opacity=Saturate(None=LinearInterpolate(A=@n53, B=@n39, Alpha=@n54))
   FadeDistance=ScalarParameter() {defaultValue=3, parameterName=DepthFade}
  )
  n56 = ShadingPathSwitch(Default=@n55, Deferred=@n55)
  n57 = SmoothStep(
   Min=ScalarParameter() {defaultValue=0.1, parameterName=ViewBasedFadeMin, sliderMax=2, sliderMin=-2}
   Max=ScalarParameter() {defaultValue=0.8, parameterName=ViewBasedFadeMax, sliderMax=2, sliderMin=-2}
   Value=Power(Base=@n51, Exp=ScalarParameter() {defaultValue=2, parameterName=ViewBasedFadeSensitivity, sliderMax=16})
  )
  n58 = LinearInterpolate(A=Multiply(A=@n56, B=Clamp(None=@n57)), Alpha=TwoSidedSign()) {constB=0}
  n59 = CameraVectorWS()
  n60 = VertexNormalWS()
  n61 = Power(Base=Abs(None=DotProduct(A=@n59, B=@n60)))
  Multiply(
   A=Multiply(A=@n12.A, B=LinearInterpolate(A=@n56, B=@n58, Alpha=@n11))
   B=LinearInterpolate(B=Saturate(None=@n61), Alpha=ScalarParameter() {defaultValue=0.3, parameterName=LerpAngleFade}) {constA=1}
  )
  Power(Base=@n42) {constExponent=1.2}
  Round(None=@n17)
  Saturate(None=MaterialFunctionCall("FadeSensitivity (optional)"=ScalarParameter() {parameterName=FadeBasedOnView}) {materialFunction=&FadeBasedOnViewAngle}.Result)
 ]
}
