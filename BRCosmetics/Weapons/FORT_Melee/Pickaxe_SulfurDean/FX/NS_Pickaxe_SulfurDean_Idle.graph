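# Idle Niagara effect for the SulfurDean pickaxe: a flame, sparse fire sprites and soft glow that follow a weapon socket.
# CoreFlame bursts four looping meshes, varied by spawn group and visibility tag, and stores a smoothed position, likely for material-driven lag.
# FireParticle sprites rise and shrink, with zero size when System.IsRunning is set; Glow aligns to socket rotation.
# Front-end flags adjust offsets, material parameters and glow; time of day dims brightness.
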NiagaraSystem /BRCosmetics/Weapons/FORT_Melee/Pickaxe_SulfurDean/FX/NS_Pickaxe_SulfurDean_Idle
dependencies {
 usedDataInterfaces [
  &NiagaraDataInterfaceVectorCurve
  &NiagaraDataInterfaceCurve
  &NiagaraDataInterfaceArrayFloat
  &NiagaraDataInterfaceColorCurve
  &NiagaraDataInterfaceSocketReader
 ]
 usedDynamicInputs [
  &Vector2DFromFloat
  &SelectFloatFromArray
  &Lerp_Float
  &SetBoolByFloatComparison
  &VectorLength
 ]
 usedModules [
  &self:SystemSpawnScript.SystemScriptSource.SystemScriptGraph.NiagaraNodeAssignment_1.SetVariables_#1
  &SystemState
  &self:SystemSpawnScript.SystemScriptSource.SystemScriptGraph.NiagaraNodeAssignment_0.SetVariables_#2
 ]
}
emitters {
 CoreFlame {
  emitterUpdateScript [
   &EmitterState(
    Inactive Response=::NewEnumerator1
    Life Cycle Mode=::NewEnumerator1
    Loop Behavior=::NewEnumerator1
    Loop Duration Mode=::NewEnumerator1
    UseLoopDelay=false
   )
   &SpawnBurst_Instantaneous(Spawn Count=4, Use Spawn Probability=false, UseLoopCountLimit=false)
  ]
  particleSpawnScript [
   &FN_SetSpawnGroup()
   &InitializeParticle(
    Color=<rgba 0.02 0.0003 0 0.6>
    Lifetime=&RandomRangeFloat(Minimum=1, Maximum=2)
    Mesh Scale=&MergeHelper_VectorSwitch(
     0=<0.423 0.423 0.5877>
     1=<0.25 0.25 0.438>
     2=<0.09 0.09 0.02>
     3=<0.27 0.27 0.438>
     Index=&SpawnGroup()
     Num=::NewEnumerator2
    )
    Mesh Scale Mode=::NewEnumerator4
    Position Offset=&MergeHelper_VectorSwitch(
     0=<0 -13.1933 -10.2658>
     1=<0 -13.1933 -9.8355>
     2=<0 -13.1933 -10.2658>
     3=?
     Index=&SpawnGroup()
     Num=::NewEnumerator1
    )
    UsePositionOffset=false
    Write Material Random=true
   )
   &InitialMeshOrientation(Mesh Orientation Mode=::NewEnumerator2, Rotation=<0 0 0.25>, Use Rotation Vector=true)
   &DynamicMaterialParameters(
    Index 0 Param 1=0.2
    Index 0 Param 2=&MergeHelper_FloatSwitch(0=0.7, 1=0, 2=0.4, 3=?, Index=&SpawnGroup(), Num=::NewEnumerator1)
    Index 0 Param 3=0.3
    Index 0 Param 4=&MakeCustomFloatFromBool(Bool=$User.bIsFrontEndPreview, True Float=0, False Float=1)
    Index 3 Param 2=&RandomRangeFloat(Minimum=0.9, Maximum=1.1)
    Index 3 Param 3=1
    Index 3 Param 4=1
    Module.Index0Float0Write=true
    Module.Index0Float1Write=true
    Module.Index0Float2Write=true
    Module.Index0Float3Write=true
    Module.Index3Float0Write=true
    Module.Index3Float1Write=true
    Module.Index3Float2Write=true
    Module.Index3Float3Write=true
    Write Parameter Index 3=::NewEnumerator1
   )
   &CameraOffset(Camera Offset Amount=&MergeHelper_FloatSwitch(0=0.2, 1=0, 2=5, 3=-0.2, Index=&SpawnGroup(), Num=::NewEnumerator2)) {module=CameraOffset001}
   /SetVariables(Particles.VisibilityTag=&SpawnGroup()) {isSetParameters=true}
   /SetVariables(Particles.BowlSize:Vector3f=<0.09 0.09 0.02>) {isSetParameters=true}
  ]
  particleUpdateScript [
   &ParticleState(
    Kill Particles When Lifetime Has Elapsed=false
    Let Infinitely Lived Particles Die When Emitter Deactivates=true
    Loop Particles Lifetime=true
   )
   /SetVariables(Particles.ColorGlow:LinearColor=&MultiplyLinearColor(
     Linear Color=<rgba 1 0.7446 0.2969 1>
     Scale Factor=&MakeVector4(
      X=1
      Y=1
      Z=1
      W=&Multiply_Float(
       A=&FNTimeOfDayScalar(Night Scale=0.6, Morning Scale=0.9, Evening Scale=0.8)
       B=&MergeHelper_FloatSwitch(0=0, 1=0, 2=0.07, 3=0.706, Index=&SpawnGroup(), Num=::NewEnumerator2)
      )
     )
    )) {isSetParameters=true}
   /SetVariables(Particles.Color002=&MultiplyLinearColor(
     Linear Color=&MergeHelper_ColorSwitch(
      0=<rgba 1 0.7446 0.2969 1>
      1=<rgba 1 0.7446 0.2969 1>
      2=<rgba 1 0.7446 0.2969 1>
      3=?
      Index=&SpawnGroup()
      Num=::NewEnumerator1
     )
     Scale Factor=&MakeVector4(X=1, Y=1, Z=1, W=&FN_FrontEndFloatModifier(Gameplay=1, Front End=1))
    )) {isSetParameters=true}
   &InterpolateOverTime(Rate Of Change=20, Target Vector Value:NiagaraPosition=$Particles.Position, Variable Type=::NewEnumerator2)
   /SetVariables(Particles.Interpolate Dynamic=&MakeFloat4FromVectorAndFloat("Vector (RGB)"=$Particles.InterpolateOverTime.InterpolateOverTime_V3_.Moving Average, "Float (Alpha)"=0.7)) {isSetParameters=true}
   &ScaleMeshSize(Scale Factor=<0.56 0.51 0.352>) {module=ScaleMeshSize002}
   &ScaleMeshSize(Scale Factor=<1.25 1.25 1.25>)
   &SocketLocation(
    SamplePosOffset=true
    Sampled Position Offset=&FN_FrontEndVector3Modifier(
     Gameplay=&MergeHelper_VectorSwitch(0=<0 0.8 3.3426>, 1=<0 0.4 3.3426>, 2=<0 0 2.7139>, 3=<0 0.4 3.3426>, Index=&SpawnGroup(), Num=::NewEnumerator2)
     Front End=&MergeHelper_VectorSwitch(
      0=<0 0.3 3.3426>
      1=<0 -0.1 3.3426>
      2=<0 0 2.7139>
      3=<0 -0.1 3.3426>
      Index=&SpawnGroup()
      Num=::NewEnumerator2
     )
    )
    Socket Reader=$System.Socket Reader
    Socket Sampling Mode=::NewEnumerator2
    VelocitySpeedLimit=false
   )
  ]
  rendererClasses [&NiagaraMeshRendererProperties]
  renderers [
   {rendererClass=&NiagaraMeshRendererProperties, rendererIndex=0}
   {rendererClass=&NiagaraMeshRendererProperties, rendererIndex=1}
   {rendererClass=&NiagaraMeshRendererProperties, rendererIndex=2}
   {rendererClass=&NiagaraMeshRendererProperties, rendererIndex=3}
  ]
 }
 FireParticle {
  emitterUpdateScript [
   &EmitterState(Inactive Response=::NewEnumerator1, Life Cycle Mode=::NewEnumerator1, Loop Duration=1, UseLoopDelay=false)
   &SpawnRate(Spawn Probability=0.6, SpawnRate=3.6, Use Spawn Probability=true)
  ]
  particleSpawnScript [
   &InitializeParticle(
    Color=&RandomRangeLinearColor(Minimum=<rgba 1 1 1 1>, Maximum=<rgba 1 1 1 2>)
    Lifetime Max=0.76
    Lifetime Min=0.6
    Lifetime Mode=::NewEnumerator1
    Sprite Rotation Mode=::NewEnumerator2
    Sprite Size Max=&Multiply_Vector2D_ByFloat(Float=&MakeCustomFloatFromBool(Bool=$System.IsRunning, True Float=0, False Float=1), Vector2D=<25 25>)
    Sprite Size Min=&Multiply_Vector2D_ByFloat(Float=&MakeCustomFloatFromBool(Bool=$System.IsRunning, True Float=0, False Float=1), Vector2D=<15 15>)
    Sprite Size Mode=::NewEnumerator2
    UsePositionOffset=false
    Write Material Random=true
   )
   &CameraOffset(Camera Offset Amount=1)
   &DynamicMaterialParameters(
    Index 0 Param 1=&RandomRangeFloat(Minimum=0, Maximum=1)
    Index 0 Param 2=&RandomRangeFloat(Minimum=0, Maximum=1)
    Index 0 Param 3=0.3
    Index 0 Param 4=&MakeCustomFloatFromBool(Bool=$User.bIsFrontEndPreview, True Float=0, False Float=1)
    Index 3 Param 2=&RandomRangeFloat(Minimum=0.9, Maximum=1.1)
    Index 3 Param 3=1
    Index 3 Param 4=1
    Module.Index0Float0Write=true
    Module.Index0Float1Write=true
    Module.Index0Float2Write=true
    Module.Index0Float3Write=true
    Module.Index3Float0Write=true
    Module.Index3Float1Write=true
    Module.Index3Float2Write=true
    Module.Index3Float3Write=true
    Write Parameter Index 3=::NewEnumerator1
   ) {module=DynamicMaterialParameters001}
  ]
  particleUpdateScript [
   &ParticleState()
   &SpriteRotationRate(Rotation Rate=&RandomRangeFloat(Minimum=-240, Maximum=240))
   &SocketLocation(
    SamplePosOffset=true
    Sampled Position Offset=&Add_Vector(
     A=&RandomRangeVector(Minimum=<-3.5 -3.5 3>, Maximum=<3.5 3.5 8>)
     B=&VectorFromCurve(
      VectorCurve=DataInterface() {propertyValues {XCurve {keys [<at 0 0 RCIM_Cubic>,<at 1 1>]}, YCurve {keys [<at 0 0 RCIM_Cubic>,<at 1 1>]}, ZCurve {keys [<at 0 0 RCIM_Cubic>,<at 1 1>]}}}
      Scale Curve=&Multiply_Vector(A=&RandomRangeVector(Minimum=<0 6 11>, Maximum=<0 8 13>), B=<1.3 1.3 1.3>)
     )
    )
    Socket Reader=$System.Socket Reader
    Socket Sampling Mode=::NewEnumerator2
    VelocitySpeedLimit=false
   ) {module=SocketLocation001}
   &ScaleSpriteSize(
    Uniform Curve Scale=0.25
    Uniform Curve Sprite Scale=DataInterface() {propertyValues {Curve {keys [
        {arriveTangent=1.701, interpMode=RCIM_Cubic, time=0, value=0.6492}
        {arriveTangent=-0.6492, interpMode=RCIM_Cubic, leaveTangent=-0.6492, time=0.6431, value=1}
        <at 1 0>
       ]}}}
   )
   &ScaleColor(
    Scale Alpha=&FloatFromCurve(
     FloatCurve=DataInterface() {propertyValues {Curve {keys [
         <at 0 0 RCIM_Cubic>
         {arriveTangent=0.6135, interpMode=RCIM_Cubic, leaveTangent=0.6135, time=0.2, value=1}
         {arriveTangent=-0.1495, interpMode=RCIM_Cubic, leaveTangent=-0.1495, tangentMode=RCTM_User, time=1, value=0.6135}
        ]}}}
     Scale Curve=3
    )
    Scale RGB=&MakeVectorFromLinearColorRGB(Color=<rgba 1 0.7446 0.2969 1>)
    ScaleA=true
    ScaleRGB=true
   )
   &ScaleColor(Scale Alpha=&FN_FrontEndFloatModifier(Gameplay=1, Front End=1), ScaleA=true, ScaleRGB=true) {module=ScaleColor001}
  ]
  rendererClasses [&NiagaraSpriteRendererProperties]
  renderers [{rendererClass=&NiagaraSpriteRendererProperties}]
 }
 Glow {
  emitterUpdateScript [
   &EmitterState(
    Inactive Response=::NewEnumerator1
    Life Cycle Mode=::NewEnumerator1
    Loop Behavior=::NewEnumerator1
    Loop Duration Mode=::NewEnumerator1
    UseLoopDelay=false
   )
   &SpawnBurst_Instantaneous(Use Spawn Probability=false, UseLoopCountLimit=false)
  ]
  particleSpawnScript [
   &InitializeParticle(
    Color=<rgba 2.1 0.5371 0 0.4087>
    Sprite Size=<30.5652 39.0606>
    Sprite Size Mode=::NewEnumerator4
    UsePositionOffset=false
    Write Material Random=true
   )
   &CameraOffset(Camera Offset Amount=1)
  ]
  particleUpdateScript [
   &ParticleState(Kill Particles When Lifetime Has Elapsed=false, Let Infinitely Lived Particles Die When Emitter Deactivates=true)
   &ScaleColor(
    Scale Alpha=&Multiply_Float(
     A=&FN_FrontEndFloatModifier(Gameplay=0.22, Front End=0.1)
     B=&FNTimeOfDayScalar(Night Scale=0.6, Morning Scale=0.9, Evening Scale=0.8)
    )
    ScaleA=true
    ScaleRGB=true
   )
   &SocketLocation(
    SamplePosOffset=true
    Sampled Position Offset=<0 0.6134 7.5326>
    Socket Reader=$System.Socket Reader
    Socket Sampling Mode=::NewEnumerator2
    VelocitySpeedLimit=false
   )
   &SpriteFacingAndAlignment(
    Sprite Alignment=&RotateVectorByQuaternion(Vector To Rotate=<0 0.3061 1>, Quaternion=$Output.SocketLocation.SampledRotation)
    WriteAlignment=true
    WriteFacing=true
   )
  ]
  rendererClasses [&NiagaraSpriteRendererProperties]
  renderers [{rendererClass=&NiagaraSpriteRendererProperties}]
 }
}
summary {
 systemName=NS_Pickaxe_SulfurDean_Idle
 emitters [
  {emitterName=FireParticle, rendererClasses [&NiagaraSpriteRendererProperties]}
  {emitterName=Glow, rendererClasses [&NiagaraSpriteRendererProperties]}
  {emitterName=CoreFlame, rendererClasses [&NiagaraMeshRendererProperties]}
 ]
 userVariables [
  User.bIsFrontEnd {type=NiagaraBool}
  User.bIsFrontEndPreview {type=NiagaraBool}
  User.bIsSwinging {type=NiagaraBool}
  User.FXint32 {type=NiagaraInt32}
  User.IsBuildInVersion {type=NiagaraBool}
 ]
}
