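# Physics asset for the SulfurDean pickaxe mesh, defining collision and joints for its dyn_ paper bones.
# Six capsule bodies form two separate groups, upper and lower.
# The upper paper bones chain out to the top edge piece, while the second lower paper and low edge piece both join the first lower paper.
# All four constraints share narrow swing limits and zero twist limits.
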PhysicsAsset /BRCosmetics/Weapons/FORT_Melee/Pickaxe_SulfurDean/Meshes/SulfurDean_Axe_Physics
bodies {
 dyn_edgepaperLow_01 [{shapeType=Capsule, length=3, radius=1, center=<7.2342 0 0>, rotation=<rot 90 0 0>}]
 dyn_edgepaperTop_01 [{shapeType=Capsule, length=5, radius=1, center=<2.5 0 0>, rotation=<rot 90 0 0>}]
 dyn_paper_01 [{shapeType=Capsule, length=1.5, radius=2.5, center=<5 0 0>}]
 dyn_paper_02 [{shapeType=Capsule, length=1.5, radius=3, center=<4.5 0 0>}]
 dyn_paperlow_01 [{shapeType=Capsule, length=3, radius=2, center=<2.2245 0 0>}]
 dyn_paperlow_02 [{shapeType=Capsule, length=0, radius=3, center=<3.557 0 0>}]
}
constraints [
 {bone1Name=dyn_paper_02, bone2Name=dyn_paper_01, swing1LimitDegrees=15, swing2LimitDegrees=15, twistLimitDegrees=0}
 {bone1Name=dyn_edgepaperTop_01, bone2Name=dyn_paper_02, swing1LimitDegrees=15, swing2LimitDegrees=15, twistLimitDegrees=0}
 {bone1Name=dyn_paperlow_02, bone2Name=dyn_paperlow_01, swing1LimitDegrees=15, swing2LimitDegrees=15, twistLimitDegrees=0}
 {bone1Name=dyn_edgepaperLow_01, bone2Name=dyn_paperlow_01, swing1LimitDegrees=15, swing2LimitDegrees=15, twistLimitDegrees=0}
]
